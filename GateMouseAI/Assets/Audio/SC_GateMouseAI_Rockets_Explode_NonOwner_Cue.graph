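# Sound cue for GateMouseAI rocket explosions heard by non-owners, capped by explosion concurrency rules that stop the oldest instance.
# It mixes close, medium and distant layers, each with its own attenuation, modulator and sound class override, with a sweetener added to the close layer.
# Quality level nodes choose between full random wave pools and smaller three-wave pools, the latter excluded from branch culling.
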SoundCue /GateMouseAI/Assets/Audio/SC_GateMouseAI_Rockets_Explode_NonOwner_Cue
root=Mixer_4
nodes [
 {
  id=Attenuation_0
  children [QualityLevel_2]
  properties {
   attenuationSettings=/Game/Sounds/Attenuation/Fort_Explosion_Distant_Athena_Att
   attenuationOverrides {falloffDistance=3000, attenuationShapeExtents {x=1000}}
  }
 }
 {
  id=Attenuation_4
  children [Mixer_0]
  properties {attenuationSettings=/Game/Sounds/Attenuation/Fort_Explosion_Close_Athena_Att, attenuationOverrides {falloffDistance=1600}}
 }
 {
  id=Attenuation_5
  children [QualityLevel_1]
  properties {
   attenuationSettings=/Game/Sounds/Attenuation/Fort_Explosion_Medium_Athena_Att
   attenuationOverrides {falloffDistance=3000, attenuationShapeExtents {x=1000}}
  }
 }
 {id=Mixer_0, children [QualityLevel_3, QualityLevel_4], properties {inputVolume [1,1.8]}}
 {id=Mixer_4, children [SoundClass_3,SoundClass_2,SoundClass_0], properties {inputVolume [1,0.7,0.5]}}
 {id=Modulator_0, children [Attenuation_4]}
 {id=Modulator_1, children [Attenuation_0], properties {pitchMax=1.1, pitchMin=0.9}}
 {id=Modulator_5, children [Attenuation_5], properties {pitchMax=1.1, pitchMin=0.9}}
 {id=QualityLevel_1, children [Random_4,Random_4,Random_5,Random_5]}
 {id=QualityLevel_2, children [Random_6,Random_6,Random_7,Random_7]}
 {id=QualityLevel_3, children [Random_8,Random_8,Random_9,Random_9]}
 {id=QualityLevel_4, children [Random_1,Random_1,Random_2,Random_2]}
 {
  id=Random_1
  children [WavePlayer_0, WavePlayer_1, WavePlayer_4, WavePlayer_12, WavePlayer_13, WavePlayer_14]
  properties {weights [1,1,1,1,1,1]}
 }
 {id=Random_2, children [WavePlayer_12, WavePlayer_13, WavePlayer_14], properties {bShouldExcludeFromBranchCulling=true, weights [1,1,1]}}
 {
  id=Random_4
  children [WavePlayer_11, WavePlayer_18, WavePlayer_19, WavePlayer_20, WavePlayer_21, WavePlayer_22, WavePlayer_23, WavePlayer_24]
  properties {weights [1,1,1,1,1,1,1,1]}
 }
 {id=Random_5, children [WavePlayer_22, WavePlayer_23, WavePlayer_24], properties {bShouldExcludeFromBranchCulling=true, weights [1,1,1]}}
 {
  id=Random_6
  children [WavePlayer_25, WavePlayer_26, WavePlayer_27, WavePlayer_28, WavePlayer_29, WavePlayer_30, WavePlayer_31, WavePlayer_32]
  properties {weights [1,1,1,1,1,1,1,1]}
 }
 {id=Random_7, children [WavePlayer_30, WavePlayer_31, WavePlayer_32], properties {bShouldExcludeFromBranchCulling=true, weights [1,1,1]}}
 {
  id=Random_8
  children [WavePlayer_33, WavePlayer_34, WavePlayer_35, WavePlayer_36, WavePlayer_37, WavePlayer_38, WavePlayer_39, WavePlayer_40]
  properties {weights [1,1,1,1,1,1,1,1]}
 }
 {id=Random_9, children [WavePlayer_38, WavePlayer_39, WavePlayer_40], properties {bShouldExcludeFromBranchCulling=true, weights [1,1,1]}}
 {
  id=SoundClass_0
  children [Modulator_1]
  properties {soundClassOverride=&Fort_WeaponRanged_Fire_Distant}
 }
 {
  id=SoundClass_2
  children [Modulator_5]
  properties {soundClassOverride=&Fort_WeaponRanged_Fire_Distant}
 }
 {id=SoundClass_3, children [Modulator_0], properties {soundClassOverride=&Fort_WeaponRanged_Fire_3P_PSM}}
 {id=WavePlayer_0, properties {soundWaveAssetPtr=&/GateMouseAI_Rockets_ExplodeSweetener_02}}
 {id=WavePlayer_1, properties {soundWaveAssetPtr=&/GateMouseAI_Rockets_ExplodeSweetener_03}}
 {id=WavePlayer_11, properties {soundWaveAssetPtr=&/GateMouseAI_Rockets_Explode_Medium_01}}
 {id=WavePlayer_12, properties {soundWaveAssetPtr=&/GateMouseAI_Rockets_ExplodeSweetener_05}}
 {id=WavePlayer_13, properties {soundWaveAssetPtr=&/GateMouseAI_Rockets_ExplodeSweetener_06}}
 {id=WavePlayer_14, properties {soundWaveAssetPtr=&/GateMouseAI_Rockets_ExplodeSweetener_01}}
 {id=WavePlayer_18, properties {soundWaveAssetPtr=&/GateMouseAI_Rockets_Explode_Medium_02}}
 {id=WavePlayer_19, properties {soundWaveAssetPtr=&/GateMouseAI_Rockets_Explode_Medium_03}}
 {id=WavePlayer_20, properties {soundWaveAssetPtr=&/GateMouseAI_Rockets_Explode_Medium_04}}
 {id=WavePlayer_21, properties {soundWaveAssetPtr=&/GateMouseAI_Rockets_Explode_Medium_05}}
 {id=WavePlayer_22, properties {soundWaveAssetPtr=&/GateMouseAI_Rockets_Explode_Medium_06}}
 {id=WavePlayer_23, properties {soundWaveAssetPtr=&/GateMouseAI_Rockets_Explode_Medium_07}}
 {id=WavePlayer_24, properties {soundWaveAssetPtr=&/GateMouseAI_Rockets_Explode_Medium_08}}
 {id=WavePlayer_25, properties {soundWaveAssetPtr=&/GateMouseAI_Rockets_Explode_Distant_01}}
 {id=WavePlayer_26, properties {soundWaveAssetPtr=&/GateMouseAI_Rockets_Explode_Distant_02}}
 {id=WavePlayer_27, properties {soundWaveAssetPtr=&/GateMouseAI_Rockets_Explode_Distant_03}}
 {id=WavePlayer_28, properties {soundWaveAssetPtr=&/GateMouseAI_Rockets_Explode_Distant_04}}
 {id=WavePlayer_29, properties {soundWaveAssetPtr=&/GateMouseAI_Rockets_Explode_Distant_05}}
 {id=WavePlayer_30, properties {soundWaveAssetPtr=&/GateMouseAI_Rockets_Explode_Distant_06}}
 {id=WavePlayer_31, properties {soundWaveAssetPtr=&/GateMouseAI_Rockets_Explode_Distant_07}}
 {id=WavePlayer_32, properties {soundWaveAssetPtr=&/GateMouseAI_Rockets_Explode_Distant_08}}
 {id=WavePlayer_33, properties {soundWaveAssetPtr=&/GateMouseAI_Rockets_Explode_Close_01}}
 {id=WavePlayer_34, properties {soundWaveAssetPtr=&/GateMouseAI_Rockets_Explode_Close_02}}
 {id=WavePlayer_35, properties {soundWaveAssetPtr=&/GateMouseAI_Rockets_Explode_Close_03}}
 {id=WavePlayer_36, properties {soundWaveAssetPtr=&/GateMouseAI_Rockets_Explode_Close_04}}
 {id=WavePlayer_37, properties {soundWaveAssetPtr=&/GateMouseAI_Rockets_Explode_Close_05}}
 {id=WavePlayer_38, properties {soundWaveAssetPtr=&/GateMouseAI_Rockets_Explode_Close_06}}
 {id=WavePlayer_39, properties {soundWaveAssetPtr=&/GateMouseAI_Rockets_Explode_Close_07}}
 {id=WavePlayer_4, properties {soundWaveAssetPtr=&/GateMouseAI_Rockets_ExplodeSweetener_04}}
 {id=WavePlayer_40, properties {soundWaveAssetPtr=&/GateMouseAI_Rockets_Explode_Close_08}}
]
properties {
 maxDistance=27500
 volumeMultiplier=0.6
 concurrencyOverrides {maxCount=4, resolutionRule=StopOldest, volumeScale=0.5}
 concurrencySet [/Game/Sounds/Concurrency_Rules/Weapon_Explosion_Conc]
}
