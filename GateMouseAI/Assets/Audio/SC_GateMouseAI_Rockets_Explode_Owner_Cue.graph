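# Sound cue for GateMouseAI rocket explosions (Owner variant), mixing close, medium and distant layers, each with its own attenuation, modulator and sound class.
# Close plays explosion reports alone, while medium and distant blend rocket explosions with their own reports plus quieter close reports.
# Quality-level nodes switch each random wave pool between a full set and a smaller subset.
# A concurrency override caps instances at two, stopping the oldest.
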SoundCue /GateMouseAI/Assets/Audio/SC_GateMouseAI_Rockets_Explode_Owner_Cue
root=Mixer_4
nodes [
 {
  id=Attenuation_0
  children [Mixer_2]
  properties {
   attenuationSettings=/Game/Sounds/Attenuation/Fort_Explosion_Distant_Athena_Att
   attenuationOverrides {falloffDistance=3000, attenuationShapeExtents {x=1000}}
  }
 }
 {
  id=Attenuation_4
  children [Mixer_0]
  properties {attenuationSettings=/Game/Sounds/Attenuation/Fort_Explosion_Close_Athena_Att, attenuationOverrides {falloffDistance=1600}}
 }
 {
  id=Attenuation_5
  children [Mixer_1]
  properties {
   attenuationSettings=/Game/Sounds/Attenuation/Fort_Explosion_Medium_Athena_Att
   attenuationOverrides {falloffDistance=3000, attenuationShapeExtents {x=1000}}
  }
 }
 {id=Mixer_0, children [QualityLevel_6, null], properties {inputVolume [2,1]}}
 {id=Mixer_1, children [QualityLevel_1, QualityLevel_5, QualityLevel_7], properties {inputVolume [1,1,0.5]}}
 {id=Mixer_2, children [QualityLevel_2, QualityLevel_8, QualityLevel_9], properties {inputVolume [1,0.5,0.1]}}
 {id=Mixer_4, children [SoundClass_3,SoundClass_2,SoundClass_0], properties {inputVolume [1,0.7,0.5]}}
 {id=Modulator_0, children [Attenuation_4]}
 {id=Modulator_1, children [Attenuation_0], properties {pitchMax=1.1, pitchMin=0.9}}
 {id=Modulator_5, children [Attenuation_5], properties {pitchMax=1.1, pitchMin=0.9}}
 {id=QualityLevel_1, children [Random_4,Random_4,Random_5,Random_5]}
 {id=QualityLevel_2, children [Random_6,Random_6,Random_7,Random_7]}
 {id=QualityLevel_5, children [Random_10,Random_10,Random_11,Random_11]}
 {id=QualityLevel_6, children [Random_12,Random_12,Random_13,Random_13]}
 {id=QualityLevel_7, children [Random_14,Random_14,Random_15,Random_15]}
 {id=QualityLevel_8, children [Random_16,Random_16,Random_17,Random_17]}
 {id=QualityLevel_9, children [Random_19,Random_19,Random_20,Random_20]}
 {id=Random_10, children [WavePlayer_50, WavePlayer_52, WavePlayer_53, WavePlayer_54], properties {weights [1,1,1,1]}}
 {id=Random_11, children [WavePlayer_52, WavePlayer_54], properties {bShouldExcludeFromBranchCulling=true, weights [1,1]}}
 {id=Random_12, children [WavePlayer_36, WavePlayer_37, WavePlayer_38, WavePlayer_39], properties {weights [1,1,1,1]}}
 {id=Random_13, children [WavePlayer_37, WavePlayer_39], properties {bShouldExcludeFromBranchCulling=true, weights [1,1]}}
 {id=Random_14, children [WavePlayer_40, WavePlayer_41, WavePlayer_42, WavePlayer_43], properties {weights [1,1,1,1]}}
 {id=Random_15, children [WavePlayer_41, WavePlayer_43], properties {bShouldExcludeFromBranchCulling=true, weights [1,1]}}
 {id=Random_16, children [WavePlayer_44, WavePlayer_45, WavePlayer_46, WavePlayer_47], properties {weights [1,1,1,1]}}
 {id=Random_17, children [WavePlayer_45, WavePlayer_47], properties {bShouldExcludeFromBranchCulling=true, weights [1,1]}}
 {id=Random_19, children [WavePlayer_56, WavePlayer_57, WavePlayer_58, WavePlayer_59], properties {weights [1,1,1,1]}}
 {id=Random_20, children [WavePlayer_57, WavePlayer_59], properties {bShouldExcludeFromBranchCulling=true, weights [1,1]}}
 {
  id=Random_4
  children [WavePlayer_11, WavePlayer_18, WavePlayer_19, WavePlayer_20, WavePlayer_21, WavePlayer_22, WavePlayer_23, WavePlayer_24]
  properties {weights [1,1,1,1,1,1,1,1]}
 }
 {id=Random_5, children [WavePlayer_22, WavePlayer_23, WavePlayer_24], properties {bShouldExcludeFromBranchCulling=true, weights [1,1,1]}}
 {
  id=Random_6
  children [WavePlayer_25, WavePlayer_26, WavePlayer_27, WavePlayer_28, WavePlayer_29, WavePlayer_30, WavePlayer_31, WavePlayer_32]
  properties {weights [1,1,1,1,1,1,1,1]}
 }
 {id=Random_7, children [WavePlayer_30, WavePlayer_31, WavePlayer_32], properties {bShouldExcludeFromBranchCulling=true, weights [1,1,1]}}
 {
  id=SoundClass_0
  children [Modulator_1]
  properties {soundClassOverride=&Fort_WeaponRanged_Fire_Distant}
 }
 {
  id=SoundClass_2
  children [Modulator_5]
  properties {soundClassOverride=&Fort_WeaponRanged_Fire_Distant}
 }
 {id=SoundClass_3, children [Modulator_0], properties {soundClassOverride=&Fort_WeaponRanged_Fire_3P_PSM}}
 {id=WavePlayer_11, properties {soundWaveAssetPtr=&/GateMouseAI_Rockets_Explode_Medium_01}}
 {id=WavePlayer_18, properties {soundWaveAssetPtr=&/GateMouseAI_Rockets_Explode_Medium_02}}
 {id=WavePlayer_19, properties {soundWaveAssetPtr=&/GateMouseAI_Rockets_Explode_Medium_03}}
 {id=WavePlayer_20, properties {soundWaveAssetPtr=&/GateMouseAI_Rockets_Explode_Medium_04}}
 {id=WavePlayer_21, properties {soundWaveAssetPtr=&/GateMouseAI_Rockets_Explode_Medium_05}}
 {id=WavePlayer_22, properties {soundWaveAssetPtr=&/GateMouseAI_Rockets_Explode_Medium_06}}
 {id=WavePlayer_23, properties {soundWaveAssetPtr=&/GateMouseAI_Rockets_Explode_Medium_07}}
 {id=WavePlayer_24, properties {soundWaveAssetPtr=&/GateMouseAI_Rockets_Explode_Medium_08}}
 {id=WavePlayer_25, properties {soundWaveAssetPtr=&/GateMouseAI_Rockets_Explode_Distant_01}}
 {id=WavePlayer_26, properties {soundWaveAssetPtr=&/GateMouseAI_Rockets_Explode_Distant_02}}
 {id=WavePlayer_27, properties {soundWaveAssetPtr=&/GateMouseAI_Rockets_Explode_Distant_03}}
 {id=WavePlayer_28, properties {soundWaveAssetPtr=&/GateMouseAI_Rockets_Explode_Distant_04}}
 {id=WavePlayer_29, properties {soundWaveAssetPtr=&/GateMouseAI_Rockets_Explode_Distant_05}}
 {id=WavePlayer_30, properties {soundWaveAssetPtr=&/GateMouseAI_Rockets_Explode_Distant_06}}
 {id=WavePlayer_31, properties {soundWaveAssetPtr=&/GateMouseAI_Rockets_Explode_Distant_07}}
 {id=WavePlayer_32, properties {soundWaveAssetPtr=&/GateMouseAI_Rockets_Explode_Distant_08}}
 {id=WavePlayer_36, properties {soundWaveAssetPtr=&/GateMouseAI_ExplosionReport_Close_01}}
 {id=WavePlayer_37, properties {soundWaveAssetPtr=&/GateMouseAI_ExplosionReport_Close_02}}
 {id=WavePlayer_38, properties {soundWaveAssetPtr=&/GateMouseAI_ExplosionReport_Close_03}}
 {id=WavePlayer_39, properties {soundWaveAssetPtr=&/GateMouseAI_ExplosionReport_Close_04}}
 {id=WavePlayer_40, properties {soundWaveAssetPtr=&/GateMouseAI_ExplosionReport_Medium_01}}
 {id=WavePlayer_41, properties {soundWaveAssetPtr=&/GateMouseAI_ExplosionReport_Medium_02}}
 {id=WavePlayer_42, properties {soundWaveAssetPtr=&/GateMouseAI_ExplosionReport_Medium_03}}
 {id=WavePlayer_43, properties {soundWaveAssetPtr=&/GateMouseAI_ExplosionReport_Medium_04}}
 {id=WavePlayer_44, properties {soundWaveAssetPtr=&/GateMouseAI_ExplosionReport_Distant_01}}
 {id=WavePlayer_45, properties {soundWaveAssetPtr=&/GateMouseAI_ExplosionReport_Distant_02}}
 {id=WavePlayer_46, properties {soundWaveAssetPtr=&/GateMouseAI_ExplosionReport_Distant_03}}
 {id=WavePlayer_47, properties {soundWaveAssetPtr=&/GateMouseAI_ExplosionReport_Distant_04}}
 {id=WavePlayer_50, properties {soundWaveAssetPtr=&GateMouseAI_ExplosionReport_Close_01}}
 {id=WavePlayer_52, properties {soundWaveAssetPtr=&GateMouseAI_ExplosionReport_Close_02}}
 {id=WavePlayer_53, properties {soundWaveAssetPtr=&GateMouseAI_ExplosionReport_Close_03}}
 {id=WavePlayer_54, properties {soundWaveAssetPtr=&GateMouseAI_ExplosionReport_Close_04}}
 {id=WavePlayer_56, properties {soundWaveAssetPtr=&GateMouseAI_ExplosionReport_Close_01}}
 {id=WavePlayer_57, properties {soundWaveAssetPtr=&GateMouseAI_ExplosionReport_Close_02}}
 {id=WavePlayer_58, properties {soundWaveAssetPtr=&GateMouseAI_ExplosionReport_Close_03}}
 {id=WavePlayer_59, properties {soundWaveAssetPtr=&GateMouseAI_ExplosionReport_Close_04}}
]
properties {
 bOverrideConcurrency=true
 maxDistance=27500
 volumeMultiplier=0.6
 concurrencyOverrides {maxCount=2, resolutionRule=StopOldest, volumeScale=0.5}
 concurrencySet [/Game/Sounds/Concurrency_Rules/Weapon_Explosion_Conc]
}
