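# Sound cue for the in-air loop sound of GateMouseAI rockets, with its own attenuation and an owner-limited concurrency override.
# A quality switch feeds either a random pick of seven loop waves (first two levels) or of the last three (other two levels).
# The pick passes through a pitch modulator, a distant weapon-fire sound class override and a smoothed, intensified Doppler node at the root.
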SoundCue /GateMouseAI/Assets/Audio/SC_GateMouseAI_Rockets_InAir
root=Doppler_1
nodes [
 {id=Doppler_1, children [SoundClass_6], properties {bUseSmoothing=true, dopplerIntensity=3}}
 {id=Modulator_7, children [QualityLevel_0], properties {pitchMax=1.1}}
 {id=QualityLevel_0, children [Random_0,Random_0,Random_1,Random_1]}
 {
  id=Random_0
  children [WavePlayer_0, WavePlayer_1, WavePlayer_9, WavePlayer_10, WavePlayer_11, WavePlayer_12, WavePlayer_13]
  properties {weights [1,1,1,1,1,1,1]}
 }
 {id=Random_1, children [WavePlayer_11, WavePlayer_12, WavePlayer_13], properties {bShouldExcludeFromBranchCulling=true, weights [1,1,1]}}
 {
  id=SoundClass_6
  children [Modulator_7]
  properties {soundClassOverride=&Fort_WeaponRanged_Fire_Distant}
 }
 {id=WavePlayer_0, properties {soundWaveAssetPtr=&/GateMouseAI_Rockets_InAir_Loop_01}}
 {id=WavePlayer_1, properties {soundWaveAssetPtr=&/GateMouseAI_Rockets_InAir_Loop_02}}
 {id=WavePlayer_10, properties {soundWaveAssetPtr=&/GateMouseAI_Rockets_InAir_Loop_04}}
 {id=WavePlayer_11, properties {soundWaveAssetPtr=&/GateMouseAI_Rockets_InAir_Loop_05}}
 {id=WavePlayer_12, properties {soundWaveAssetPtr=&/GateMouseAI_Rockets_InAir_Loop_06}}
 {id=WavePlayer_13, properties {soundWaveAssetPtr=&/GateMouseAI_Rockets_InAir_Loop_07}}
 {id=WavePlayer_9, properties {soundWaveAssetPtr=&/GateMouseAI_Rockets_InAir_Loop_03}}
]
properties {
 attenuationSettings=/GateMouseAI/Assets/Audio/Settings/ATT_GateMouseAI_Rockets_InAir
 bOverrideConcurrency=true
 maxDistance=10512
 pitchMultiplier=1.5
 soundClassObject=&Fort_WeaponRanged_Fire_3P_PSM
 volumeMultiplier=0.4
 attenuationOverrides {falloffDistance=5000, attenuationShapeExtents {x=1000}}
 concurrencyOverrides {bLimitToOwner=true, maxCount=2, volumeScale=0.5}
}
