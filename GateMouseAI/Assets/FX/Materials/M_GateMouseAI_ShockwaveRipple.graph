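# Translucent unlit Niagara sprite and mesh material for the GateMouseAI shockwave ripple effect, using pixel-normal-offset refraction.
# Polar-mapped Perlin noise, panned over time with explicit UV derivatives, distorts a clamped flare texture's UVs.
# Dynamic parameters control flare scale and distortion strength.
# The particle-alpha-weighted mask, depth-faded except on low/mobile, drives a refraction-style 1–1.1 lerp, alongside a sphere normal map and particle-colored emissive.
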Material /GateMouseAI/Assets/FX/Materials/M_GateMouseAI_ShockwaveRipple
graph {
 properties {
  bAutomaticallySetUsageInEditor=false
  bUsedWithNiagaraMeshParticles=true
  bUsedWithNiagaraSprites=true
  blendMode=BLEND_Translucent
  pixelDepthOffsetMode=PDOM_Legacy
  refractionMethod=RM_PixelNormalOffset
  shadingModel=MSM_Unlit
  usedShadingModels=Unlit
 }
 nodes [
  n1 = ParticleColor()
  n2 = DynamicParameter() {paramNames [AnimOffset,UVDistortion,Param3,Param4]}
  n3 = TextureCoordinate()
  n4 = Constant() {r=0.1}
  n5 = Constant() {r=1}
  n6 = Time()
  n7 = ScalarParameter() {defaultValue=1, parameterName=PanSpeed}
  n8 = TextureSample!(
   UVs=Panner(
    Coordinate=MaterialFunctionCall!(UV Coord=@n3, LenghtScale=@n4, RadialScale=@n5) {materialFunction=/Game/Global/MaterialFunctions/TechArt/MF_PolarCoordinates}.Result
    Time=Multiply(A=@n6, B=@n7)
   ) {speedX=-0.3}
   Tex
   "DDX(UVs)"=DDX(Value=@n3)
   "DDY(UVs)"=DDY(Value=@n3)
   Apply View MipBias
  ) {
   mipValueMode=TMVM_Derivative
   samplerType=SAMPLERTYPE_LinearGrayscale
   texture=&T_SimplePerlin_Greyscale
  }
  n9 = LinearInterpolate(Alpha=Desaturation(None=@n8.RGB, Fraction=Constant() {r=1})) {constA=-1}
  n10 = Constant2Vector() {g=0.1, r=0.1}
  n11 = Add(
   A=MaterialFunctionCall(UVs=TextureCoordinate(), Texture Scale=@n2.AnimOffset) {materialFunction=&ScaleUVsByCenter}.UVs
   B=Multiply(A=@n2.UVDistortion, B=Multiply(A=@n9, B=@n10))
  )
  n12 = Multiply(
   A=@n1.A
   B=Saturate(None=TextureSample!(UVs=@n11, Tex, Apply View MipBias) {samplerSource=SSM_Clamp_WorldGroupSettings, texture=&/T_GateMouseAI_Shockwave_Flare}.R)
  )
  n13 = DepthFade() {fadeDistanceDefault=25}
  LinearInterpolate(Alpha=MaterialFunctionCall!(Default=Multiply(A=@n12, B=@n13), Low/Mobile=@n12) {materialFunction=/Game/Effects/Fort_Effects/Materials/_Functions/MF_LowMobile_PathingSwitch}.Result) {constA=1, constB=1.1}
  MaterialFunctionCall(Emissive=@n1.RGB) {materialFunction=&MF_FX_EmissiveAdjustments}
  TextureSample() {samplerType=SAMPLERTYPE_Normal, texture=&T_Sphere_N}
 ]
}
