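# Niagara screen-effect system for GateMouseAI that plays a brief, one-shot Vignette, with no user parameters.
# A single emitter bursts one scaled mesh particle after a short delay, drawn by a mesh renderer rather than sprites.
# Dynamic material parameters animate it through a lifetime curve and oscillating waveforms.
# Alpha rises quickly then fades out, with a random multiplier blended in over its life to vary the fade.
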NiagaraSystem /GateMouseAI/Assets/FX/NS_GateMouseAI_ScreenFX
dependencies {usedDataInterfaces [&NiagaraDataInterfaceCurve], usedDynamicInputs [], usedModules [&SystemState]}
emitters {Vignette {
  emitterUpdateScript [
   &EmitterState(
    Inactive Response=::NewEnumerator1
    Life Cycle Mode=::NewEnumerator1
    Loop Behavior=::NewEnumerator1
    Loop Duration Mode=::NewEnumerator1
    UseLoopDelay=false
   )
   &SpawnBurst_Instantaneous(Spawn Count=1, Spawn Time=0.2, Use Spawn Probability=false)
  ]
  particleSpawnScript [&InitializeParticle(
    Color=<rgba 1 1 1 0.75>
    Lifetime=0.8
    Mass Mode=::NewEnumerator1
    Mesh Scale=<100 100 100>
    Mesh Scale Mode=::NewEnumerator4
    Position Offset=<-250 0 0>
    Sprite Size Mode=::NewEnumerator4
    UsePositionOffset=false
    Write Material Random=true
   ) {module=InitializeParticle001}]
  particleUpdateScript [
   &ParticleState()
   &SolveForcesAndVelocity(Clamp Velocity=false, Limit Acceleration=false)
   &DynamicMaterialParameters(
    Index 0 Param 1=&FloatFromCurve(FloatCurve=DataInterface() {propertyValues {Curve {keys [<at 0 0.3>, <at 0.05 0>, {leaveTangentWeight=0.1667, time=0.5, value=0}, {arriveTangentWeight=0.2406, time=1, value=0.3}]}}})
    Index 0 Param 3=&Waveform("Waveform [1]"=::NewEnumerator8, "[1] Frequency"=5, "[1] Amplitude Min/Max"=<0.65 0.75>)
    Index 0 Param 4=&Waveform("Waveform [1]"=::NewEnumerator8, "[1] Frequency"=5, "[1] Amplitude Min/Max"=<-1 1>, "[1] Scale"=0.1)
    Index 1 Param 1=&Waveform("Waveform [1]"=::NewEnumerator8, "[1] Frequency"=5, "[1] Amplitude Min/Max"=<-1 1>, "[1] Scale"=0.3)
    Module.Index0Float0Write=true
    Module.Index0Float1Write=true
    Module.Index0Float2Write=true
    Module.Index0Float3Write=true
    Module.Index1Float0Write=true
    Module.Index1Float1Write=false
    Module.Index1Float2Write=false
    Module.Index1Float3Write=false
    Write Parameter Index 1=::NewEnumerator1
   )
   &ScaleColor(
    Scale Alpha=&FloatFromCurve(
     FloatCurve=DataInterface() {propertyValues {Curve {
        defaultValue=1
        keys [
         {interpMode=RCIM_Cubic, tangentMode=RCTM_Break, time=0, value=0}
         {arriveTangentWeight=0.0324, interpMode=RCIM_Cubic, leaveTangentWeight=0.3563, tangentMode=RCTM_Break, time=0.1, value=1}
         {arriveTangentWeight=0.2662, interpMode=RCIM_Cubic, leaveTangentWeight=0.1936, tangentMode=RCTM_Break, time=0.2, value=1}
         {interpMode=RCIM_Cubic, leaveTangent=-0.0405, tangentMode=RCTM_Break, time=1, value=0}
        ]
       }}}
     Scale Curve=&Lerp_Float(
      A=1
      B=&RandomRangeFloat(Minimum=0.6, Maximum=1.2, Evaluation Type=::NewEnumerator1)
      Alpha=&FloatFromCurve(FloatCurve=DataInterface() {propertyValues {Curve {keys [<at 0.15 0>,<at 0.6 1>]}}})
     )
    )
    ScaleA=true
    ScaleRGB=false
   ) {module=ScaleColor001}
  ]
  rendererClasses [&NiagaraMeshRendererProperties]
  renderers [{rendererClass=&NiagaraMeshRendererProperties, rendererIndex=0}]
 }}
summary {
 systemName=NS_GateMouseAI_ScreenFX
 emitters [{emitterName=Vignette, rendererClasses [&NiagaraMeshRendererProperties]}]
 userVariables []
}
